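# Niagara ribbon trail for the DriftSwat pickaxe, drawing a pink-magenta lightning streak and glow trails behind the blade.
# Three emitters use the FN AnimTrail modules to follow bones or sockets on User.SkeletalMesh, each with three ribbon renderers.
# BladeLightnin is shorter-lived and colored by a gradient along execution index, faded by age.
# The two BladeTrail layers take time-of-day colors and shape alpha with curves, including one along execution index.
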NiagaraSystem /BRCosmetics/Weapons/FORT_Melee/Pickaxe_DriftSwat/FX/NS_Pickaxe_DriftSwat_Lightning_Trail
dependencies {
 usedDataInterfaces [&NiagaraDataInterfaceCurve, &NiagaraDataInterfaceColorCurve]
 usedDynamicInputs []
 usedModules [
  &SystemState
  &self:SystemSpawnScript.SystemScriptSource.SystemScriptGraph.NiagaraNodeAssignment_0.SetVariables_#1
 ]
}
emitters {
 BladeLightnin {
  emitterUpdateScript [
   &EmitterState()
   &FN_SpawnAnimTrail()
  ]
  particleSpawnScript [
   &FN_InitializeAnimTrailBoneIndecies(First Bone or Socket Index=1, Second Bone or Socket Index=0)
   &FN_InitializeAnimTrail(
    AnimTrailWidth=0.8
    Color=&FNTimeOfDayColor(
     Night Color=<rgba 50 0.7521 25.1492 0.2>
     Morning Color=<rgba 50 0.7521 25.1492 0.2>
     Day Color=<rgba 50 0.7521 25.1492 0.2>
     Evening Color=<rgba 50 0.7521 25.1492 0.2>
    )
    Lifetime=0.11
   )
  ]
  particleUpdateScript [
   &SolveForcesAndVelocity(Clamp Velocity=false, Limit Acceleration=false)
   &FN_UpdateAnimTrail(Locked Particle RibbonWidth=1, SkeletalMesh=$User.SkeletalMesh) {module=FN_UpdateAnimTrail001}
   &ScaleColor(
    Scale Alpha=&FloatFromCurve(
     FloatCurve=DataInterface() {propertyValues {Curve {keys [
         {arriveTangent=-2.003, interpMode=RCIM_Cubic, leaveTangent=-2.003, tangentMode=RCTM_User, time=0.5, value=0.95}
         {interpMode=RCIM_Cubic, tangentMode=RCTM_User, time=1, value=0}
        ]}}}
     Scale Curve=0.5
    )
    ScaleA=true
    ScaleRGB=true
   ) {enabled=false}
   &ScaleRibbonWidth(Ribbon Width Scale=&FloatFromCurve(FloatCurve=DataInterface() {propertyValues {Curve {keys [<at 0.5 1 RCIM_Cubic>, {arriveTangent=-3.9793, interpMode=RCIM_Cubic, leaveTangent=-3.9793, tangentMode=RCTM_User, time=1, value=0}]}}})) {enabled=false}
   /SetVariables(Particles.Color=&MultiplyLinearColor(
     Linear Color=$Particles.Color
     Scale Factor=&MakeVector4(
      X=1
      Y=1
      Z=1
      W=&Lerp_Float(
       A=&FloatFromCurve(
        FloatCurve=DataInterface() {propertyValues {Curve {keys [
            {tangentMode=RCTM_User, time=0, value=1}
            {arriveTangent=-0.0405, interpMode=RCIM_Cubic, leaveTangent=-0.0405, tangentMode=RCTM_User, time=0.7009, value=1}
            {arriveTangent=-1.3914, interpMode=RCIM_Cubic, leaveTangent=-1.3914, tangentMode=RCTM_User, time=1, value=0}
           ]}}}
        CurveIndex=&ReturnNormalizedExecIndex()
       )
       B=1
       Alpha=&FloatFromCurve(
        FloatCurve=DataInterface() {propertyValues {Curve {keys [
            {tangentMode=RCTM_User, time=0, value=1}
            <at 0.1 1>
            {arriveTangent=-32.5174, interpMode=RCIM_Cubic, leaveTangent=-32.5174, tangentMode=RCTM_User, time=0.13, value=0}
           ]}}}
        CurveIndex=$Emitter.Age
       )
      )
     )
    )) {enabled=false, isSetParameters=true, module=SetVariables_#2}
   &CameraOffset(Camera Offset Amount=-2)
   &Color(Color=&MultiplyLinearColor(
     Linear Color=&ScaleLinearColorByCurve(
      InitialValue=&ColorFromCurve(
       ColorCurve=DataInterface() {propertyValues {
         AlphaCurve {keys [
           <at 0 1 RCIM_Cubic>
           {arriveTangent=-3.0051, interpMode=RCIM_Cubic, leaveTangent=-1.1734, tangentMode=RCTM_Break, time=0.8, value=0.1}
           <at 1 0>
          ]}
         BlueCurve {keys [<at 0.2015 0.503 RCIM_Cubic>,<at 0.6849 0.672 RCIM_Cubic>]}
         GreenCurve {keys [<at 0.2015 0.015 RCIM_Cubic>,<at 0.6849 0.35 RCIM_Cubic>]}
         RedCurve {keys [<at 0.2015 1 RCIM_Cubic>,<at 0.6849 1 RCIM_Cubic>]}
        }}
       CurveIndex=&OneMinusFloat(Float=&ReturnNormalizedExecIndex())
      )
      ScaleCurve=DataInterface() {propertyValues {
        AlphaCurve {keys [<at 0.9506 1>,<at 1 0>]}
        BlueCurve {keys [<at 0.2993 4>]}
        GreenCurve {keys [<at 0.2993 4>]}
        RedCurve {keys [<at 0.2993 4>]}
       }}
      CurveIndex=$Particles.NormalizedAge
     )
     Scale Factor=&MakeFloat4FromVectorAndFloat(
      "Vector (RGB)"=<1 1 1>
      "Float (Alpha)"=&FloatFromCurve(
       FloatCurve=DataInterface() {propertyValues {Curve {keys [
           {interpMode=RCIM_Cubic, tangentMode=RCTM_User, time=0, value=0}
           {interpMode=RCIM_Cubic, tangentMode=RCTM_User, time=0.8, value=1}
           {tangentMode=RCTM_User, time=1, value=0}
          ]}}}
       CurveIndex=&ReturnNormalizedExecIndex()
      )
     )
    )) {module=Color001}
   &ScaleRibbonWidth(Ribbon Width Scale=&FloatFromCurve(
     FloatCurve=DataInterface() {propertyValues {Curve {keys [
         {arriveTangent=1.4153, interpMode=RCIM_Cubic, time=0, value=1}
         {arriveTangent=-1.946, interpMode=RCIM_Cubic, leaveTangent=-1.946, tangentMode=RCTM_User, time=1, value=0.5}
        ]}}}
     CurveIndex=&OneMinusFloat(Float=&ReturnNormalizedExecIndex())
    )) {enabled=false, module=ScaleRibbonWidth001}
  ]
  rendererClasses [&NiagaraRibbonRendererProperties]
  renderers [
   {rendererClass=&NiagaraRibbonRendererProperties}
   {rendererClass=&NiagaraRibbonRendererProperties, rendererIndex=1}
   {rendererClass=&NiagaraRibbonRendererProperties, rendererIndex=2}
  ]
 }
 BladeTrail {
  emitterUpdateScript [
   &EmitterState()
   &FN_SpawnAnimTrail()
  ]
  particleSpawnScript [
   &FN_InitializeAnimTrailBoneIndecies(First Bone or Socket Index=1, Second Bone or Socket Index=0)
   &FN_InitializeAnimTrail(
    Color=&FNTimeOfDayColor(
     Night Color=<rgba 10 0.1504 5.0298 0.06>
     Morning Color=<rgba 10 0.1504 5.0298 0.1>
     Day Color=<rgba 10 0.1504 5.0298 0.1>
     Evening Color=<rgba 10 0.1504 5.0298 0.1>
    )
    Lifetime=0.2
   )
  ]
  particleUpdateScript [
   &SolveForcesAndVelocity(Clamp Velocity=false, Limit Acceleration=false)
   &FN_UpdateAnimTrail(Locked Particle RibbonWidth=1, SkeletalMesh=$User.SkeletalMesh) {module=FN_UpdateAnimTrail001}
   &ScaleColor(
    Scale Alpha=&FloatFromCurve(
     FloatCurve=DataInterface() {propertyValues {Curve {keys [
         {arriveTangent=-2.003, interpMode=RCIM_Cubic, leaveTangent=-2.003, tangentMode=RCTM_User, time=0.5, value=0.95}
         {interpMode=RCIM_Cubic, tangentMode=RCTM_User, time=1, value=0}
        ]}}}
     Scale Curve=0.5
    )
    ScaleA=true
    ScaleRGB=true
   )
   /SetVariables(Particles.Color=&MultiplyLinearColor(
     Linear Color=$Particles.Color
     Scale Factor=&MakeVector4(
      X=1
      Y=1
      Z=1
      W=&Lerp_Float(
       A=&FloatFromCurve(
        FloatCurve=DataInterface() {propertyValues {Curve {keys [
            {tangentMode=RCTM_User, time=0, value=1}
            {arriveTangent=-0.0405, interpMode=RCIM_Cubic, leaveTangent=-0.0405, tangentMode=RCTM_User, time=0.7009, value=1}
            {arriveTangent=-1.3914, interpMode=RCIM_Cubic, leaveTangent=-1.3914, tangentMode=RCTM_User, time=1, value=0}
           ]}}}
        CurveIndex=&ReturnNormalizedExecIndex()
       )
       B=1
       Alpha=&FloatFromCurve(
        FloatCurve=DataInterface() {propertyValues {Curve {keys [
            {tangentMode=RCTM_User, time=0, value=1}
            <at 0.1 1>
            {arriveTangent=-32.5174, interpMode=RCIM_Cubic, leaveTangent=-32.5174, tangentMode=RCTM_User, time=0.13, value=0}
           ]}}}
        CurveIndex=$Emitter.Age
       )
      )
     )
    )) {enabled=false, isSetParameters=true, module=SetVariables_#2}
   &CameraOffset(Camera Offset Amount=-2)
   &ScaleColor(
    Scale Alpha=&FloatFromCurve(
     FloatCurve=DataInterface() {propertyValues {Curve {keys [
         {arriveTangent=4.6221, interpMode=RCIM_Cubic, leaveTangent=4.6221, tangentMode=RCTM_User, time=0.1572, value=0}
         {interpMode=RCIM_Cubic, tangentMode=RCTM_User, time=0.6, value=1}
         {arriveTangent=-5.1569, tangentMode=RCTM_User, time=1, value=0}
        ]}}}
     CurveIndex=&ReturnNormalizedExecIndex()
    )
    ScaleA=true
    ScaleRGB=true
   ) {module=ScaleColor001}
  ]
  rendererClasses [&NiagaraRibbonRendererProperties]
  renderers [
   {rendererClass=&NiagaraRibbonRendererProperties}
   {rendererClass=&NiagaraRibbonRendererProperties, rendererIndex=1}
   {rendererClass=&NiagaraRibbonRendererProperties, rendererIndex=2}
  ]
 }
 BladeTrail002 {
  emitterUpdateScript [
   &EmitterState()
   &FN_SpawnAnimTrail()
  ]
  particleSpawnScript [
   &FN_InitializeAnimTrailBoneIndecies()
   &FN_InitializeAnimTrail(
    Color=&FNTimeOfDayColor(
     Night Color=<rgba 2 0.0301 1.006 0.1>
     Morning Color=<rgba 2 0.0301 1.006 0.3>
     Day Color=<rgba 2 0.0301 1.006 0.3>
     Evening Color=<rgba 2 0.0301 1.006 0.3>
    )
    Lifetime=0.2
   )
  ]
  particleUpdateScript [
   &SolveForcesAndVelocity(Clamp Velocity=false, Limit Acceleration=false)
   &FN_UpdateAnimTrail(Locked Particle RibbonWidth=1, SkeletalMesh=$User.SkeletalMesh) {module=FN_UpdateAnimTrail001}
   &ScaleColor(
    Scale Alpha=&FloatFromCurve(
     FloatCurve=DataInterface() {propertyValues {Curve {keys [
         {arriveTangent=-2.003, interpMode=RCIM_Cubic, leaveTangent=-2.003, tangentMode=RCTM_User, time=0.5, value=0.95}
         {interpMode=RCIM_Cubic, tangentMode=RCTM_User, time=1, value=0}
        ]}}}
     Scale Curve=0.5
    )
    ScaleA=true
    ScaleRGB=true
   )
   /SetVariables(Particles.Color=&MultiplyLinearColor(
     Linear Color=$Particles.Color
     Scale Factor=&MakeVector4(
      X=1
      Y=1
      Z=1
      W=&Lerp_Float(
       A=&FloatFromCurve(
        FloatCurve=DataInterface() {propertyValues {Curve {keys [
            {tangentMode=RCTM_User, time=0, value=1}
            {arriveTangent=-0.0405, interpMode=RCIM_Cubic, leaveTangent=-0.0405, tangentMode=RCTM_User, time=0.7009, value=1}
            {arriveTangent=-1.3914, interpMode=RCIM_Cubic, leaveTangent=-1.3914, tangentMode=RCTM_User, time=1, value=0}
           ]}}}
        CurveIndex=&ReturnNormalizedExecIndex()
       )
       B=1
       Alpha=&FloatFromCurve(
        FloatCurve=DataInterface() {propertyValues {Curve {keys [
            {tangentMode=RCTM_User, time=0, value=1}
            <at 0.1 1>
            {arriveTangent=-32.5174, interpMode=RCIM_Cubic, leaveTangent=-32.5174, tangentMode=RCTM_User, time=0.13, value=0}
           ]}}}
        CurveIndex=$Emitter.Age
       )
      )
     )
    )) {enabled=false, isSetParameters=true, module=SetVariables_#2}
   &CameraOffset(Camera Offset Amount=-2)
   &ScaleColor(
    Scale Alpha=&FloatFromCurve(
     FloatCurve=DataInterface() {propertyValues {Curve {keys [
         {arriveTangent=4.6221, interpMode=RCIM_Cubic, leaveTangent=4.6221, tangentMode=RCTM_User, time=0.1572, value=0}
         {interpMode=RCIM_Cubic, tangentMode=RCTM_User, time=0.6, value=1}
         {arriveTangent=-5.1569, tangentMode=RCTM_User, time=1, value=0}
        ]}}}
     CurveIndex=&ReturnNormalizedExecIndex()
    )
    ScaleA=true
    ScaleRGB=true
   ) {module=ScaleColor001}
  ]
  rendererClasses [&NiagaraRibbonRendererProperties]
  renderers [
   {rendererClass=&NiagaraRibbonRendererProperties}
   {rendererClass=&NiagaraRibbonRendererProperties, rendererIndex=1}
   {rendererClass=&NiagaraRibbonRendererProperties, rendererIndex=2}
  ]
 }
}
summary {
 systemName=NS_Pickaxe_DriftSwat_Lightning_Trail
 emitters [
  {emitterName=BladeTrail002, rendererClasses [&NiagaraRibbonRendererProperties]}
  {emitterName=BladeLightnin, rendererClasses [&NiagaraRibbonRendererProperties]}
  {emitterName=BladeTrail, rendererClasses [&NiagaraRibbonRendererProperties]}
 ]
 userVariables [
  User.Color A {type=LinearColor, description="", defaultValue {struct=LinearColor, value=<rgba 0 0.5019 1 1>}}
  User.ColorSwitcher {type=NiagaraFloat, defaultValue=0, description=""}
  User.MeshColors {
   type=NiagaraDataInterfaceArrayColor
   description=""
   defaultValue {
    struct=NiagaraExt_VariableValue_DataInterface
    value {
     dataInterface=&self:NiagaraDataInterfaceArrayColor_0
     dataInterfaceClass=&NiagaraDataInterfaceArrayColor
    }
   }
  }
  User.MeshColors_Variant_01 {
   type=NiagaraDataInterfaceArrayColor
   description=""
   defaultValue {
    struct=NiagaraExt_VariableValue_DataInterface
    value {
     dataInterface=&self:NiagaraDataInterfaceArrayColor_1
     dataInterfaceClass=&NiagaraDataInterfaceArrayColor
    }
   }
  }
  User.RibbonIDOffset {type=NiagaraInt32}
  User.SkeletalMesh {
   type=NiagaraDataInterfaceSkeletalMesh
   defaultValue {
    struct=NiagaraExt_VariableValue_DataInterface
    value {dataInterface=&self:NiagaraDataInterfaceSkeletalMesh_0}
   }
  }
 ]
}
