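# Niagara idle effect for the DriftSwat pickaxe, adding a pink glow, flickering light and electric sparks.
# Glow fades in and dims in the front end; Light flickers via curves and waveforms, and both scale with time of day.
# Lightning001 randomly bursts mesh arcs, gated by User.ParticlesAmount, along one of two lines between sampled skeletal-mesh sockets, re-snapping each update.
# The older Lightning emitter and Empty emitter are disabled.
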NiagaraSystem /BRCosmetics/Weapons/FORT_Melee/Pickaxe_DriftSwat/FX/NS_Pickaxe_DriftSwat_Spark_Idle
dependencies {
 usedDataInterfaces [&NiagaraDataInterfaceArrayFloat3, &NiagaraDataInterfaceColorCurve, &NiagaraDataInterfaceCurve]
 usedDynamicInputs [
  &SelectVectorFromArray
  &AddVectorToPosition
  &MakeFloatFromInt
  &Boolean_NotOperation
 ]
 usedModules [
  &self:SystemSpawnScript.SystemScriptSource.SystemScriptGraph.NiagaraNodeAssignment_1.SetVariables_#1
  &SystemState
  &FN_Timer
  &self:SystemSpawnScript.SystemScriptSource.SystemScriptGraph.NiagaraNodeAssignment_0.SetVariables_#2
  &self:SystemSpawnScript.SystemScriptSource.SystemScriptGraph.NiagaraNodeAssignment_2.SetVariables_#3
 ]
}
emitters {
 Empty {
  bEnabled=false
  emitterUpdateScript [
   &EmitterState()
   &SpawnBurst_Instantaneous(Use Spawn Probability=false, UseLoopCountLimit=false)
  ]
  particleSpawnScript [&InitializeParticle(Color Mode=::NewEnumerator0, Lifetime=5, UsePositionOffset=false, Write Material Random=true)]
  particleUpdateScript [&ParticleState()]
  rendererClasses [&NiagaraSpriteRendererProperties, &NiagaraComponentRendererProperties]
  renderers [
   {rendererClass=&NiagaraSpriteRendererProperties}
   {rendererClass=&NiagaraComponentRendererProperties, rendererIndex=1}
  ]
 }
 Glow {
  emitterUpdateScript [
   &EmitterState(Life Cycle Mode=::NewEnumerator1, Loop Behavior=::NewEnumerator1, Loop Duration Mode=::NewEnumerator1, UseLoopDelay=false)
   &SpawnBurst_Instantaneous(Spawn Probability=HlslExpression() {hlslExpression=1.0}, Use Spawn Probability=false, UseLoopCountLimit=true)
  ]
  particleSpawnScript [&InitializeParticle(
    Color=<rgba 0.8632 0.013 0.4342 0.3>
    Position Offset=<0 0 25>
    Sprite Size Mode=::NewEnumerator3
    Uniform Sprite Size=50
    UsePositionOffset=true
    Write Material Random=true
   )]
  particleUpdateScript [
   &ParticleState(Kill Particles When Lifetime Has Elapsed=false, Let Infinitely Lived Particles Die When Emitter Deactivates=true)
   &ScaleColor(
    Scale Alpha=&Lerp_Float(
     A=0
     B=&FNTimeOfDayScalar(Night Scale=0.5, Morning Scale=0.5, Evening Scale=0.5)
     Alpha=&FloatFromCurve(
      FloatCurve=DataInterface() {propertyValues {Curve {keys [
          {interpMode=RCIM_Cubic, tangentMode=RCTM_User, time=0, value=0}
          {arriveTangent=2.4173, interpMode=RCIM_Cubic, leaveTangent=2.4173, tangentMode=RCTM_User, time=1, value=1}
         ]}}}
      CurveIndex=$Emitter.Age
      Scale Curve=&FN_FrontEndFloatModifier(Gameplay=1, Front End=0.5)
     )
    )
    Scale RGB=&VectorFromFloat(Value=&FN_FrontEndFloatModifier(Gameplay=1, Front End=0.5))
    ScaleA=true
    ScaleRGB=true
   )
  ]
  rendererClasses [&NiagaraSpriteRendererProperties]
  renderers [{rendererClass=&NiagaraSpriteRendererProperties}]
 }
 Light {
  emitterSpawnScript [&FN_PreserveSkydiveUserVariables() {enabled=false, module=PreserveUserVariables}]
  emitterUpdateScript [
   &EmitterState()
   &SpawnBurst_Instantaneous(Use Spawn Probability=false, UseLoopCountLimit=true)
  ]
  particleSpawnScript [
   &PartitionParticles() {enabled=false}
   &InitializeParticle(Lifetime=2, Position Offset=<5 0 25>, Sprite Size Mode=::NewEnumerator3, UsePositionOffset=true, Write Material Random=true)
  ]
  particleUpdateScript [
   &ParticleState(
    Kill Particles When Lifetime Has Elapsed=false
    Let Infinitely Lived Particles Die When Emitter Deactivates=true
    Loop Particles Lifetime=true
   )
   &FN_Light_Attributes(
    Color=<rgba 0.8632 0.013 0.4342 1>
    Create Color Attribute=true
    Create Exponent Attribute=true
    Create Radius Attribute=true
    Exponent=2
    Radius=100
    Scale Color=&Multiply_Float(
     A=&FNTimeOfDayScalar(Night Scale=2.5, Morning Scale=3, Day Scale=4, Evening Scale=3)
     B=&Waveform(
      Waveform Count=::NewEnumerator2
      "Waveform [1]"=::NewEnumerator0
      "[1] Frequency"=1
      "[1] Amplitude Min/Max"=<0.9 1>
      "[1] Scale"=0.5
      "[2] Average"=false
      "Waveform [2]"=::NewEnumerator5
      "[2] Input Position"=0
      "[2] Phase"=$Emitter.Age
      "[2] Phase Scale"=1
      "[2] Compound Frequency"=2
      "[2] Frequency"=2
      "[2] Amplitude Min/Max"=<0 1>
      "[2] Scale"=0.05
      "[2] Bias"=0
      "[2] Exponent"=1
      "[2] Pulse Duration"=0.3
      "[2] Clip Triangle Min/Max"=<0 0>
      "[2] Random Seed"=0
      "[3] Average"=false
      "Waveform [3]"=::NewEnumerator5
      "[3] Input Position"=0
      "[3] Phase"=$Emitter.Age
      "[3] Phase Scale"=1
      "[3] Frequency"=1
      "[3] Compound Frequency"=2
      "[3] Amplitude Min/Max"=<0 1>
      "[3] Scale"=0.03
      "[3] Bias"=0
      "[3] Exponent"=1
      "[3] Pulse Duration"=0.3
      "[3] Clip Triangle Min/Max"=<0 0>
      "[3] Random Seed"=0
     )
    )
   )
   /SetVariables(Particles.LightIntensity:NiagaraFloat=&Multiply_Float(
     A=50
     B=&Multiply_Float(
      A=&Multiply_Float(
       A=&FNTimeOfDayScalar(Night Scale=0.6, Morning Scale=1, Evening Scale=1)
       B=&FloatFromCurve(FloatCurve=DataInterface() {propertyValues {Curve {keys [
            {arriveTangent=9.0725, interpMode=RCIM_Cubic, time=0, value=0.5}
            {arriveTangent=0.25, interpMode=RCIM_Cubic, leaveTangent=0.25, time=0.1408, value=0.9847}
            {arriveTangent=-0.1661, interpMode=RCIM_Cubic, leaveTangent=-0.1661, time=0.3, value=0.575}
            {arriveTangent=-0.5, interpMode=RCIM_Cubic, leaveTangent=-0.5, time=0.5, value=0.925}
            {arriveTangent=-0.0833, interpMode=RCIM_Cubic, leaveTangent=-0.0833, time=0.6, value=0.425}
            {arriveTangent=0.1875, interpMode=RCIM_Cubic, leaveTangent=0.1875, time=0.8, value=0.9}
            {interpMode=RCIM_Cubic, leaveTangent=-0.0405, time=1, value=0.5}
           ]}}})
      )
      B=&Waveform(
       Waveform Count=::NewEnumerator1
       "Waveform [1]"=::NewEnumerator0
       "[1] Frequency"=1
       "[1] Amplitude Min/Max"=<0.9 1>
       "[1] Scale"=0.5
       "[2] Average"=false
       "[2] Input Position"=0
       "[2] Phase"=$Emitter.Age
       "[2] Phase Scale"=1
       "[2] Compound Frequency"=2
       "[2] Frequency"=2
       "[2] Amplitude Min/Max"=<0 1>
       "[2] Scale"=0.3
       "[2] Bias"=0
       "[2] Exponent"=1
       "[2] Pulse Duration"=0.1
       "[2] Clip Triangle Min/Max"=<0 0>
       "[2] Random Seed"=0
      )
     )
    )) {isSetParameters=true}
  ]
  rendererClasses [
   &NiagaraLightRendererProperties
   &NiagaraSpriteRendererProperties
   &NiagaraComponentRendererProperties
  ]
  renderers [
   {rendererClass=&NiagaraLightRendererProperties, rendererIndex=0}
   {rendererClass=&NiagaraSpriteRendererProperties, rendererIndex=1}
   {rendererClass=&NiagaraComponentRendererProperties, rendererIndex=2}
  ]
 }
 Lightning {
  bEnabled=false
  emitterUpdateScript [
   &EmitterState(
    Inactive Response=::NewEnumerator2
    Life Cycle Mode=::NewEnumerator1
    Loop Duration=&RandomRangeFloat(Minimum=0.1, Maximum=0.22)
    Recalculate Duration Each Loop=true
    UseLoopDelay=false
   )
   &SpawnRate(SpawnRate=900, Use Spawn Probability=false) {enabled=false, module=SpawnRate001}
   &SpawnRate(SpawnGroup=1, Use Spawn Probability=false)
   &SpawnBurst_Instantaneous(
    Spawn Count=&RandomRangeInt(Minimum=1, Maximum=1)
    Spawn Time=&RandomRangeFloat(Minimum=0, Maximum=0.1)
    Use Spawn Probability=false
    UseLoopCountLimit=false
   )
   /SetVariables(Emitter.DepthFade=35) {isSetParameters=true, module=SetVariables_#4}
  ]
  particleSpawnScript [
   &PartitionParticles(Number Of Partitions=3) {enabled=false}
   /SetVariables(Particles.MeshIndex=&RandomRangeInt(Minimum=0, Maximum=5)) {isSetParameters=true, module=SetVariables_#5}
   /SetVariables(
    Particles.SetWidth:NiagaraFloat=&RandomRangeFloat(Minimum=2.3, Maximum=2.8)
    Particles.width rand:NiagaraFloat=&RandomRangeFloat(Minimum=0.1, Maximum=5)
   ) {isSetParameters=true, module=SetVariables_#6}
   /SetVariables(Particles.MeshIndex=5) {enabled=false, isSetParameters=true}
   &InitializeParticle(
    Color=&MultiplyLinearColor(
     Linear Color=&FNTimeOfDayColor(
      Night Color=<rgba 50 0.7521 25.1492 1>
      Morning Color=<rgba 50 0.7521 25.1492 1>
      Day Color=<rgba 50 0.7521 25.1492 1>
      Evening Color=<rgba 50 0.7521 25.1492 1>
     )
     Scale Factor=&MakeFloat4FromVectorAndFloat("Vector (RGB)"=&VectorFromFloat(Value=0.008), "Float (Alpha)"=1)
    )
    Lifetime=&RandomRangeFloat(Minimum=0.8, Maximum=0.5)
    Mesh Scale Mode=::NewEnumerator3
    Mesh Uniform Scale=&Multiply_Float(
     A=&MakeCustomFloatFromBool(
      Bool=&SetBoolByIntComparison(A=$Emitter.SpawnGroup, B=0, Comparison Type=::NewEnumerator4)
      True Float=&RandomRangeFloat(Minimum=0.9, Maximum=2)
      False Float=&RandomRangeFloat(Minimum=0.5, Maximum=0.5)
     )
     B=0.5
    )
    Sprite Rotation Mode=::NewEnumerator2
    Sprite Size Mode=::NewEnumerator3
    Uniform Sprite Size=5
    UsePositionOffset=true
    Write Material Random=false
   )
   &ShapeLocation(
    Cylinder Height=0
    Cylinder Radius=4
    MaskSpawnGroup=false
    Offset Coordinate Space=::Local
    Offset Mode=::NewEnumerator0
    Override Seed=false
    Shape Primitive=::NewEnumerator1
    Surface Only=false
    UseOwnerScale=true
   ) {enabled=false}
   &CameraOffset(Camera Offset Amount=10)
   /SetVariables(Particles.ElectricityVisiblityTag:NiagaraInt32=&MakeIntFromFloat(Float=$User.ColorSwitcher)) {enabled=false, isSetParameters=true}
   &DynamicMaterialParameters(
    Index 0 Param 2=$Particles.width rand
    Index 0 Param 4=3
    Module.Param0WriteEnabled=false
    Module.Param0WriteEnabled1=true
    Module.Param1WriteEnabled=true
    Module.Param1WriteEnabled1=true
    Module.Param2WriteEnabled=true
    Module.Param2WriteEnabled1=true
    Module.Param3WriteEnabled=false
    Module.Param3WriteEnabled1=true
    Write Parameter Index 1=true
   )
   &InitialMeshOrientation(Use Rotation Vector=true)
   &ShapeLocation(
    Cylinder Distribution=::NewEnumerator2
    Cylinder Height=&Multiply_Float(A=5.25, B=0)
    Cylinder Radius=12
    MaskSpawnGroup=false
    Non Uniform Scale=<1.15 1 1>
    Offset=<-10 0 24>
    Offset Mode=::NewEnumerator0
    Rotation Axis=<0 1 0>
    Rotation Mode=::NewEnumerator1
    Shape Primitive=::NewEnumerator1
    U Position=&RandomRangeFloat(Minimum=0.65, Maximum=1.1)
    UseOwnerScale=true
    V Position=&UniformAOrBFloat(A=0)
   ) {module=ShapeLocation001}
   &ShapeLocation(MaskSpawnGroup=false, Offset=<0 0 -12>, Offset Mode=::NewEnumerator0, Override Seed=false, Sphere Radius=1, UseOwnerScale=true) {module=ShapeLocation002}
   /SetVariables(Particles.Position=&MakeCustomVectorFromBool(
     Bool=&SetBoolByIntComparison(A=$Emitter.SpawnGroup, B=0, Comparison Type=::NewEnumerator4)
     True Vector=$Particles.ShapeLocation001.ShapeVector
     False Vector=$Particles.ShapeLocation002.ShapeVector
    )) {isSetParameters=true}
  ]
  particleUpdateScript [
   &ParticleState()
   &SolveForcesAndVelocity(Clamp Velocity=false, Limit Acceleration=false) {enabled=false}
   /SetVariables(Particles.MeshPosition:NiagaraPosition=$Particles.Position) {enabled=false, isSetParameters=true}
   &DynamicMaterialParameters(
    Index 0 Param 1=&RandomRangeFloat(Minimum=1, Maximum=0.5)
    Index 0 Param 3=$Particles.NormalizedAge
    Index 0 Param 4=1
    Index 1 Param 1=&Multiply_Float(
     A=&Add_Float(A=$System.Age, B=&RandomRangeFloat(Minimum=0, Maximum=100))
     B=0.1
    )
    Module.Param0WriteEnabled=true
    Module.Param0WriteEnabled1=true
    Module.Param1WriteEnabled=false
    Module.Param1WriteEnabled1=true
    Module.Param2WriteEnabled=true
    Module.Param2WriteEnabled1=true
    Module.Param3WriteEnabled=true
    Module.Param3WriteEnabled1=true
    Write Parameter Index 1=true
   ) {module=DynamicMaterialParameters001}
   &ScaleColor(
    Scale RGB=&VectorFromFloat(Value=&Multiply_Float(A=&Lerp_Float(A=0.133, B=1, Alpha=$System.Flicker), B=1))
    ScaleA=false
    ScaleRGB=true
   ) {enabled=false}
   /SetVariables(Particles.FlareColor=&MultiplyLinearColor(
     Linear Color=&ColorFromCurve(ColorCurve=DataInterface() {propertyValues {
        AlphaCurve {keys [<at 0 1 RCIM_Cubic>,<at 0.7365 0 RCIM_Cubic>]}
        BlueCurve {keys [<at 0 33 RCIM_Cubic>,<at 0.274 11>,<at 1 1 RCIM_Cubic>]}
        GreenCurve {keys [<at 0 0 RCIM_Cubic>,<at 0.274 1.8419>,<at 1 0.1451 RCIM_Cubic>]}
        RedCurve {keys [<at 0 19.25 RCIM_Cubic>,<at 0.274 7.4894>,<at 1 0.7863 RCIM_Cubic>]}
       }})
     Scale Factor=<1 1 1 1>
    )) {isSetParameters=true, module=SetVariables_#7}
   &ScaleColor(
    Color Value To Scale=$Particles.Color
    Scale RGB=&Multiply_VectorByFloat(Vector=<1 1 1>, Float=$System.TimeOfDayBrightnessMultiplier)
    ScaleA=true
    ScaleRGB=true
   ) {enabled=false, module=ScaleColor001}
   /SetVariables(
    Particles.CameraOffsetGlow:NiagaraFloat=&Add_Float(A=$Particles.CameraOffset, B=-0.001)
    Particles.ColorGlow:LinearColor=<rgba 10 2 5.9631 0.5>
    Particles.DynamicMaterialParameterGlow:Vector4f=&MakeVector4(X=&Add_Float(A=$Particles.SetWidth, B=1), Y=$Particles.width rand, Z=$Particles.NormalizedAge, W=0.5)
   ) {isSetParameters=true, module=SetVariables_#8}
  ]
  rendererClasses [&NiagaraMeshRendererProperties, &NiagaraSpriteRendererProperties]
  renderers [
   {rendererClass=&NiagaraMeshRendererProperties, rendererIndex=0}
   {rendererClass=&NiagaraMeshRendererProperties, rendererIndex=1}
   {rendererClass=&NiagaraSpriteRendererProperties, rendererIndex=2}
   {rendererClass=&NiagaraSpriteRendererProperties, rendererIndex=3}
   {rendererClass=&NiagaraSpriteRendererProperties, rendererIndex=4}
  ]
 }
 Lightning001 {
  emitterUpdateScript [
   &EmitterState(
    Inactive Response=::NewEnumerator2
    Life Cycle Mode=::NewEnumerator1
    Loop Duration=&RandomRangeFloat(Minimum=0.1, Maximum=0.25)
    Recalculate Duration Each Loop=true
    UseLoopDelay=false
   )
   &SpawnRate(SpawnRate=100, Use Spawn Probability=false) {enabled=false, module=SpawnRate001}
   &SpawnRate(SpawnGroup=1, Use Spawn Probability=false) {enabled=false}
   &SpawnBurst_Instantaneous(
    Spawn Count=&RandomRangeInt(Minimum=1, Maximum=1)
    Spawn Probability=$User.ParticlesAmount
    Spawn Time=&RandomRangeFloat(Minimum=0.05, Maximum=0.35)
    Use Spawn Probability=true
    UseLoopCountLimit=false
   )
   /SetVariables(Emitter.DepthFade=35) {isSetParameters=true, module=SetVariables_#4}
  ]
  particleSpawnScript [
   /SetVariables(Particles.MeshIndex=&RandomRangeInt(Minimum=0, Maximum=5)) {isSetParameters=true, module=SetVariables_#5}
   /SetVariables(
    Particles.SetWidth:NiagaraFloat=&RandomRangeFloat(Minimum=2.5, Maximum=3)
    Particles.width rand:NiagaraFloat=&RandomRangeFloat(Minimum=0.1, Maximum=5)
   ) {isSetParameters=true, module=SetVariables_#6}
   /SetVariables(
    Particles.Line:NiagaraBool=&RandomBool()
    Particles.Radn Pos 2:NiagaraFloat=&RandomRangeFloat(Minimum=0, Maximum=0.9)
    Particles.rand Pos:NiagaraFloat=&RandomRangeFloat(Minimum=0.2, Maximum=1.15)
   ) {isSetParameters=true}
   &InitializeParticle(
    Color=&MultiplyLinearColor(
     Linear Color=&FNTimeOfDayColor(
      Night Color=<rgba 50 0.7521 25.1492 1>
      Morning Color=<rgba 50 0.7521 25.1492 1>
      Day Color=<rgba 50 0.7521 25.1492 1>
      Evening Color=<rgba 50 0.7521 25.1492 1>
     )
     Scale Factor=&MakeFloat4FromVectorAndFloat("Vector (RGB)"=&VectorFromFloat(Value=0.008), "Float (Alpha)"=1)
    )
    Lifetime=&RandomRangeFloat(Minimum=0.8, Maximum=0.5)
    Mesh Scale=&Multiply_VectorByFloat(
     Vector=&RandomRangeVector(Minimum=<1 0.5 0.5>, Maximum=<1 1 1>)
     Float=&RandomRangeFloat(Minimum=1.25, Maximum=1)
    )
    Mesh Scale Mode=::NewEnumerator4
    Sprite Rotation Mode=::NewEnumerator2
    Sprite Size Mode=::NewEnumerator3
    Uniform Sprite Size=5
    UsePositionOffset=true
    Write Material Random=false
   )
   &CameraOffset(Camera Offset Amount=10)
   &DynamicMaterialParameters(
    Index 0 Param 2=$Particles.width rand
    Index 0 Param 4=3
    Module.Param0WriteEnabled=false
    Module.Param0WriteEnabled1=true
    Module.Param1WriteEnabled=true
    Module.Param1WriteEnabled1=true
    Module.Param2WriteEnabled=true
    Module.Param2WriteEnabled1=true
    Module.Param3WriteEnabled=false
    Module.Param3WriteEnabled1=true
    Write Parameter Index 1=true
   )
   &SampleSkeletalMesh(
    AdditionalYawPitchRoll=false
    Bone / Socket Index=1
    Mesh Sampling Type:ENiagaraSkelSamplingModeFull=::NewEnumerator5
    Skeletal Mesh:NiagaraDataInterfaceSkeletalMesh=$User.SkeletalMesh
    Socket Sampling Mode:ENiagaraSocketSamplingMode=::NewEnumerator5
   )
   &SampleSkeletalMesh(
    AdditionalYawPitchRoll=false
    Bone / Socket Index=2
    Mesh Sampling Type:ENiagaraSkelSamplingModeFull=::NewEnumerator5
    Skeletal Mesh:NiagaraDataInterfaceSkeletalMesh=$User.SkeletalMesh
    Socket Sampling Mode:ENiagaraSocketSamplingMode=::NewEnumerator5
   ) {module=SampleSkeletalMesh001}
   &SampleSkeletalMesh(
    AdditionalYawPitchRoll=false
    Bone / Socket Index=4
    Mesh Sampling Type:ENiagaraSkelSamplingModeFull=::NewEnumerator5
    Skeletal Mesh:NiagaraDataInterfaceSkeletalMesh=$User.SkeletalMesh
    Socket Sampling Mode:ENiagaraSocketSamplingMode=::NewEnumerator5
   ) {module=SampleSkeletalMesh002}
   &SampleSkeletalMesh(
    AdditionalYawPitchRoll=false
    Bone / Socket Index=3
    Mesh Sampling Type:ENiagaraSkelSamplingModeFull=::NewEnumerator5
    Skeletal Mesh:NiagaraDataInterfaceSkeletalMesh=$User.SkeletalMesh
    Socket Sampling Mode:ENiagaraSocketSamplingMode=::NewEnumerator5
   ) {module=SampleSkeletalMesh003}
   /SetVariables(Particles.Position=&self:ScratchDynamicInput(
     bool:NiagaraBool=$Particles.Line
     False:NiagaraPosition=&LerpPosition(A=$Particles.SampleSkeletalMesh.SampledPosition, B=$Particles.SampleSkeletalMesh001.SampledPosition, Alpha=$Particles.rand Pos)
     True:NiagaraPosition=&LerpPosition(A=$Particles.SampleSkeletalMesh002.SampledPosition, B=$Particles.SampleSkeletalMesh003.SampledPosition, Alpha=$Particles.Radn Pos 2)
    )) {isSetParameters=true}
   &InitialMeshOrientation(
    Mesh Axis To Orient=<0 0 1>
    Mesh Orientation Mode=::NewEnumerator2
    Orientation Coordinate Space=::Local
    Orientation Vector=&Lerp_Vector(
     A=&MakeCustomVectorFromBool(
      Bool=$Particles.Line
      True Vector=&Subtract_Position(A=$Particles.SampleSkeletalMesh.SampledPosition, B=$Particles.SampleSkeletalMesh001.SampledPosition)
      False Vector=&Subtract_Position(A=$Particles.SampleSkeletalMesh002.SampledPosition, B=$Particles.SampleSkeletalMesh003.SampledPosition)
     )
     B=&RandomVector(Vector Scale=1)
     Alpha=0
    )
    Rotation=&Multiply_Vector(A=&RandomRangeVector(Minimum=<1 1 1>, Maximum=<-1 -1 -1>), B=<1 0.105 0.105>)
    Use Rotation Vector=true
   )
  ]
  particleUpdateScript [
   &ParticleState()
   &DynamicMaterialParameters(
    Index 0 Param 1=&RandomRangeFloat(Minimum=1, Maximum=0.5)
    Index 0 Param 3=$Particles.NormalizedAge
    Index 0 Param 4=1
    Index 1 Param 1=&Multiply_Float(
     A=&Add_Float(A=$System.Age, B=&RandomRangeFloat(Minimum=0, Maximum=100))
     B=0.1
    )
    Module.Param0WriteEnabled=true
    Module.Param0WriteEnabled1=true
    Module.Param1WriteEnabled=false
    Module.Param1WriteEnabled1=true
    Module.Param2WriteEnabled=true
    Module.Param2WriteEnabled1=true
    Module.Param3WriteEnabled=true
    Module.Param3WriteEnabled1=true
    Write Parameter Index 1=true
   ) {module=DynamicMaterialParameters001}
   /SetVariables(Particles.Position=&self:ScratchDynamicInput(
     bool:NiagaraBool=$Particles.Line
     False:NiagaraPosition=&LerpPosition(A=$Particles.SampleSkeletalMesh.SampledPosition, B=$Particles.SampleSkeletalMesh001.SampledPosition, Alpha=$Particles.rand Pos)
     True:NiagaraPosition=&LerpPosition(A=$Particles.SampleSkeletalMesh002.SampledPosition, B=$Particles.SampleSkeletalMesh003.SampledPosition, Alpha=$Particles.Radn Pos 2)
    )) {isSetParameters=true}
   /SetVariables(Particles.FlareColor=&MultiplyLinearColor(
     Linear Color=&ColorFromCurve(ColorCurve=DataInterface() {propertyValues {
        AlphaCurve {keys [<at 0 1 RCIM_Cubic>,<at 0.7365 0 RCIM_Cubic>]}
        BlueCurve {keys [<at 0 33 RCIM_Cubic>,<at 0.274 11>,<at 1 1 RCIM_Cubic>]}
        GreenCurve {keys [<at 0 0 RCIM_Cubic>,<at 0.274 1.8419>,<at 1 0.1451 RCIM_Cubic>]}
        RedCurve {keys [<at 0 19.25 RCIM_Cubic>,<at 0.274 7.4894>,<at 1 0.7863 RCIM_Cubic>]}
       }})
     Scale Factor=<1 1 1 1>
    )) {isSetParameters=true, module=SetVariables_#7}
   /SetVariables(
    Particles.CameraOffsetGlow:NiagaraFloat=&Add_Float(A=$Particles.CameraOffset, B=-0.001)
    Particles.ColorGlow:LinearColor=<rgba 10 2 5.9631 0.5>
    Particles.DynamicMaterialParameterGlow:Vector4f=&MakeVector4(X=&Add_Float(A=$Particles.width rand, B=0.1), Y=$Particles.width rand, Z=$Particles.NormalizedAge, W=0.5)
   ) {isSetParameters=true, module=SetVariables_#8}
  ]
  rendererClasses [&NiagaraMeshRendererProperties]
  renderers [
   {rendererClass=&NiagaraMeshRendererProperties, rendererIndex=0}
   {rendererClass=&NiagaraMeshRendererProperties, rendererIndex=1}
   {rendererClass=&NiagaraMeshRendererProperties, rendererIndex=2}
  ]
 }
}
summary {
 systemName=NS_Pickaxe_DriftSwat_Spark_Idle
 emitters [
  {
   bEnabled=false
   emitterName=Lightning
   rendererClasses [&NiagaraMeshRendererProperties, &NiagaraSpriteRendererProperties]
  }
  {emitterName=Glow, rendererClasses [&NiagaraSpriteRendererProperties]}
  {
   emitterName=Light
   rendererClasses [
    &NiagaraLightRendererProperties
    &NiagaraSpriteRendererProperties
    &NiagaraComponentRendererProperties
   ]
  }
  {emitterName=Lightning001, rendererClasses [&NiagaraMeshRendererProperties]}
  {
   bEnabled=false
   emitterName=Empty
   rendererClasses [&NiagaraSpriteRendererProperties, &NiagaraComponentRendererProperties]
  }
 ]
 userVariables [
  User.bIsFrontEnd {type=NiagaraBool}
  User.bIsFrontEndPreview {type=NiagaraBool}
  User.ColorSwitcher {type=NiagaraFloat, defaultValue=0, description=""}
  User.FlickerSpeed {type=NiagaraFloat, defaultValue=6.666, description=""}
  User.MaskParticlesAmount {type=NiagaraFloat, defaultValue=0, description=""}
  User.ParticlesAmount {type=NiagaraFloat, defaultValue=1, description=""}
  User.PowersActive {type=NiagaraBool}
  User.SkeletalMesh {
   type=NiagaraDataInterfaceSkeletalMesh
   defaultValue {
    struct=NiagaraExt_VariableValue_DataInterface
    value {dataInterface=&self:NiagaraDataInterfaceSkeletalMesh_0}
   }
  }
 ]
}
