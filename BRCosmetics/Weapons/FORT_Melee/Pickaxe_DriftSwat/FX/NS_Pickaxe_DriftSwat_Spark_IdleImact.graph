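# Niagara system for the DriftSwat pickaxe, named as an idle impact spark effect, that emits a burst of glowing pink-red sprite sparks.
# The one active emitter tapers its spawn rate to zero over the loop, spawning short-lived sparks on an offset cylinder with random velocity.
# Gravity, drag and an alpha fade shape their motion and decay.
# A disabled Sparks_Eyes emitter reads particles from an Eyes_FX emitter absent from this system.
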NiagaraSystem /BRCosmetics/Weapons/FORT_Melee/Pickaxe_DriftSwat/FX/NS_Pickaxe_DriftSwat_Spark_IdleImact
dependencies {
 usedDataInterfaces [
  &NiagaraDataInterfaceCurve
  &NiagaraDataInterfaceArrayFloat3
  &NiagaraDataInterfaceParticleRead
  &NiagaraDataInterfaceVector2DCurve
 ]
 usedDynamicInputs [
  &Multiply_VectorByFloat
  &FN_FrontEndFloatModifier
  &Boolean_NotOperation
  &FNTimeOfDayScalar
 ]
 usedModules [
  &self:SystemSpawnScript.SystemScriptSource.SystemScriptGraph.NiagaraNodeAssignment_1.SetVariables_#1
  &SystemState
  &FN_Timer
  &self:SystemSpawnScript.SystemScriptSource.SystemScriptGraph.NiagaraNodeAssignment_0.SetVariables_#2
  &self:SystemSpawnScript.SystemScriptSource.SystemScriptGraph.NiagaraNodeAssignment_2.SetVariables_#3
 ]
}
emitters {
 Shoulders_FX002 {
  emitterUpdateScript [
   &EmitterState()
   &SpawnRate(
    SpawnGroup=1
    SpawnRate=&FloatFromCurve(
     FloatCurve=DataInterface() {propertyValues {Curve {keys [
         {arriveTangent=-2.2522, interpMode=RCIM_Cubic, leaveTangent=-2.2522, tangentMode=RCTM_User, time=0, value=1}
         {
          arriveTangentWeight=0.6058
          interpMode=RCIM_Cubic
          leaveTangentWeight=0.6058
          tangentMode=RCTM_User
          tangentWeightMode=RCTWM_WeightedBoth
          time=1
          value=0
         }
        ]}}}
     CurveIndex=$Emitter.NormalizedLoopAge
     Scale Curve=300
    )
    Use Spawn Probability=false
   )
  ]
  particleSpawnScript [
   &InitializeParticle(
    Color=&MultiplyLinearColor(
     Linear Color=&FNTimeOfDayColor(
      Night Color=<rgba 50 0.7521 25.1492 1>
      Morning Color=<rgba 50 0.7521 25.1492 1>
      Day Color=<rgba 50 0.7521 25.1492 1>
      Evening Color=<rgba 50 0.7521 25.1492 1>
     )
     Scale Factor=&MakeFloat4FromVectorAndFloat("Vector (RGB)"=&VectorFromFloat(Value=1), "Float (Alpha)"=1)
    )
    Lifetime=&RandomRangeFloat(Minimum=0.2, Maximum=0.35)
    Mesh Scale Mode=::NewEnumerator3
    Mesh Uniform Scale=&MakeCustomFloatFromBool(
     Bool=&SetBoolByIntComparison(A=$Emitter.SpawnGroup, B=0, Comparison Type=::NewEnumerator4)
     True Float=&RandomRangeFloat(Minimum=0.9, Maximum=2)
     False Float=&RandomRangeFloat(Minimum=0.5, Maximum=0.5)
    )
    Sprite Rotation Mode=::NewEnumerator2
    Sprite Size Max=<2 5>
    Sprite Size Min=<1 10>
    Sprite Size Mode=::NewEnumerator2
    UsePositionOffset=true
    Write Material Random=false
   )
   &CameraOffset()
   &AddVelocity(Velocity=&RandomVector(Vector Scale=&RandomRangeFloat(Minimum=100, Maximum=200)))
   &ShapeLocation(
    Cylinder Distribution=::NewEnumerator2
    Cylinder Height=5.25
    Cylinder Radius=12
    MaskSpawnGroup=false
    Non Uniform Scale=<1.15 1 1>
    Offset=<-10 0 25>
    Offset Coordinate Space=::Local
    Offset Mode=::NewEnumerator0
    Rotation Axis=<0 1 0>
    Rotation Mode=::NewEnumerator1
    Shape Primitive=::NewEnumerator1
    U Position=&RandomRangeFloat(Minimum=0.65, Maximum=0.9)
    UseOwnerScale=true
    V Position=&UniformAOrBFloat(A=0)
   ) {module=ShapeLocation001}
  ]
  particleUpdateScript [
   &ParticleState()
   &GravityForce()
   &Drag(Drag=3, Use Linear Drag=true, Use Rotational Drag=true)
   &SolveForcesAndVelocity(Clamp Velocity=false, Limit Acceleration=false)
   &ScaleColor(
    Scale Alpha=&FloatFromCurve(FloatCurve=DataInterface() {propertyValues {Curve {keys [<at 0 1>,<at 1 0>]}}})
    ScaleA=true
    ScaleRGB=true
   )
  ]
  rendererClasses [&NiagaraSpriteRendererProperties]
  renderers [{rendererClass=&NiagaraSpriteRendererProperties}]
 }
 Sparks_Eyes {
  bEnabled=false
  emitterUpdateScript [
   &EmitterState()
   &SpawnParticlesFromOtherEmitter(
    Attribute Reader=DataInterface() {propertyValues {EmitterBinding {bindingMode=Other, emitterName=Eyes_FX}}}
    Maximum Age Spawn Threshold=0.1
    Spawn Rate=25
    Spawn Rate Per Particle Cap=500
    Use Maximum Age=true
    Use Minimum Age=false
   )
  ]
  particleSpawnScript [
   &InitializeParticle(
    Color=<rgba 9 1.2188 5.1094 1>
    Lifetime=&RandomRangeFloat(Minimum=0.132, Maximum=0.5)
    Sprite Size=&Lerp_Vector2(A=<1.2 0.5>, B=<0.7 3>, Alpha=&RandomRangeFloat(Minimum=0, Maximum=1))
    Sprite Size Mode=::NewEnumerator4
    UsePositionOffset=false
    Write Material Random=true
   )
   &SampleParticlesFromOtherEmitter(Mass Application:ENiagara_LifetimeInheritanceOptions=::NewEnumerator2, Mass Sampling:ENiagara_DirectReadSamplingMode=::NewEnumerator1)
   /SetVariables(Particles.SpawnGroup=&MakeIntFromFloat(Float=$Output.SampleParticlesFromOtherEmitter.SampledMass)) {isSetParameters=true}
   &AddVelocity(
    Cone Angle=33
    Cone Axis=&SelectVectorFromArray(Direct Array Index=$Particles.SpawnGroup, Vector Selection Array=DataInterface() {propertyValues {FloatData [<-1 0 0.5>, <1 0 0.5>]}})
    Module.UseConeFalloff=false
    Module.UseVelDistribution=false
    Override Seed=false
    Velocity Mode=::NewEnumerator2
    Velocity Speed=&RandomRangeFloat(Minimum=11, Maximum=50)
   )
   /SetVariables(Particles.SubImageIndex=&RandomRangeFloat(Minimum=0, Maximum=4)) {isSetParameters=true}
  ]
  particleUpdateScript [
   &Drag(Drag=&RandomRangeFloat(Minimum=3, Maximum=6), Use Linear Drag=true, Use Rotational Drag=false)
   &GravityForce(Gravity=<0 0 -300>)
   &InheritSourceMovement(
    Applied Position Scale=&VectorFromFloat(Value=&FloatFromCurve(FloatCurve=DataInterface() {propertyValues {Curve {keys [<at 0 1 RCIM_Cubic>, {arriveTangent=-2.5875, interpMode=RCIM_Cubic, leaveTangent=-2.5875, tangentMode=RCTM_User, time=1, value=0}]}}}))
    ApplyForce=false
    ApplyPosition=true
    ApplyPositionFalloff=false
   )
   &SolveForcesAndVelocity(Clamp Velocity=false, Limit Acceleration=false)
   &ParticleState()
   &ScaleSpriteSize(
    Non-Uniform Curve Sprite Scale=DataInterface() {propertyValues {
      XCurve {keys [{tangentMode=RCTM_User, time=0, value=1}, <at 0.6 1>, {tangentMode=RCTM_User, time=1, value=0}]}
      YCurve {keys [{tangentMode=RCTM_User, time=0, value=1}, {tangentMode=RCTM_User, time=1, value=0}]}
     }}
    Scale Sprite Size Mode=::NewEnumerator3
   )
   &ScaleSpriteSizeBySpeed(Max Scale Factor=<1 3>, Velocity Threshold=66)
  ]
  rendererClasses [&NiagaraSpriteRendererProperties]
  renderers [{rendererClass=&NiagaraSpriteRendererProperties}]
 }
}
summary {
 systemName=NS_Pickaxe_DriftSwat_Spark_IdleImact
 emitters [
  {emitterName=Shoulders_FX002, rendererClasses [&NiagaraSpriteRendererProperties]}
  {bEnabled=false, emitterName=Sparks_Eyes, rendererClasses [&NiagaraSpriteRendererProperties]}
 ]
 userVariables [
  User.bIsFrontEnd {type=NiagaraBool}
  User.bIsFrontEndPreview {type=NiagaraBool}
  User.ColorSwitcher {type=NiagaraFloat, defaultValue=0, description=""}
  User.FlickerSpeed {type=NiagaraFloat, defaultValue=6.666, description=""}
  User.MaskParticlesAmount {type=NiagaraFloat, defaultValue=0, description=""}
  User.ParticlesAmount {type=NiagaraFloat, defaultValue=1, description=""}
  User.PowersActive {type=NiagaraBool}
  User.SkeletalMesh {
   type=NiagaraDataInterfaceSkeletalMesh
   defaultValue {
    struct=NiagaraExt_VariableValue_DataInterface
    value {dataInterface=&self:NiagaraDataInterfaceSkeletalMesh_0}
   }
  }
 ]
}
