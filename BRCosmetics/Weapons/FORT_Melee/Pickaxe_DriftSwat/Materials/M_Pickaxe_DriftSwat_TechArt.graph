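# Master material for the DriftSwat pickaxe, layering animated electric emissive between the 2023 character parent PreFX and PostFX functions.
# Pre-skinned local position supplies coordinates and center-line, side and pulse masks for two panning electric-noise layers and a lightning texture.
# A time-driven sine step flips between two lightning channels for flicker.
# The tinted result is limited to a selectable Tech Art Mask channel and scaled by time of day.
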Material /BRCosmetics/Weapons/FORT_Melee/Pickaxe_DriftSwat/Materials/M_Pickaxe_DriftSwat_TechArt
graph {
 properties {
  bUseMaterialAttributes=true
  bUsedWithMeshDeformer=true
  bUsedWithNanite=true
  bUsedWithSkeletalMesh=true
  pixelDepthOffsetMode=PDOM_Legacy
  shadingModel=MSM_FromMaterialExpression
  usedShadingModels=Default Lit | Subsurface | Clear Coat | Subsurface Profile | Cloth
 }
 nodes [
  n1 = NamedRerouteUsage()
  Divide(A=@n1, B=PreSkinnedLocalBounds().Extents)
  n2 = MaterialFunctionCall() {materialFunction=&MF_CharacterParent_2023_PreFX}
  n3 = GetMaterialAttributes(None=@n2.Result) {attributeGetTypes [#1]}
  n4 = Multiply(A=NamedRerouteUsage(), B=Constant2Vector() {r=0.2})
  n5 = ComponentMask(None=NamedRerouteUsage()) {b=true, r=true}
  n6 = Multiply(A=@n5, B=ScalarParameter() {defaultValue=0.01, group=TechArtEmissive, parameterName=WorldAlignedTiling01})
  n7 = VectorParameter() {group=TechArtEmissive, parameterName=FlowingSpeed01, defaultValue=<rgba -0.1 0 0 1>}
  n8 = TextureObjectParameter() {
   group=TechArtEmissive
   parameterName=BasicEmissivePattern
   samplerType=SAMPLERTYPE_LinearColor
   texture=/Game/Characters/Player/Male/Medium/Bodies/M_MED_SRDrift/Textures/T_ElectricNoise
  }
  n9 = TextureSample(UVs=Panner(Coordinate=Add(A=@n4, B=@n6), Speed=ComponentMask(None=@n7.RGB) {g=true, r=true}) {speedX=0.1}, Tex=@n8) {samplerType=SAMPLERTYPE_LinearColor, texture=&T_ElectricNoise}
  n10 = NamedRerouteUsage()
  n11 = Constant2Vector() {r=0.1}
  n12 = ScalarParameter() {defaultValue=0.03, group=TechArtEmissive, parameterName=WorldAlignedTiling02}
  n13 = MaterialFunctionCall(Float2=Add(A=Multiply(A=@n10, B=@n11), B=Multiply(A=@n5, B=@n12))) {materialFunction=&BreakOutFloat2Components}
  n14 = Panner(
   Coordinate=MaterialFunctionCall(X=@n13.R, Y=Multiply(A=@n13.G)) {materialFunction=&MakeFloat2}.Result
   Speed=ComponentMask(None=VectorParameter() {group=TechArtEmissive, parameterName=FlowingSpeed02, defaultValue=<rgba -0.08 0 0 1>}.RGB) {g=true, r=true}
  ) {speedX=0.1}
  n15 = Multiply(
   A=TextureSample(UVs=@n14, Tex=@n8) {samplerType=SAMPLERTYPE_LinearColor, texture=&T_ElectricNoise}.R
   B=ScalarParameter() {defaultValue=1.5, group=TechArtEmissive, parameterName=BasePatternMulti}
  )
  n16 = TextureSample(UVs=@n14) {samplerType=SAMPLERTYPE_LinearColor, texture=/BRCosmetics/Weapons/FORT_Melee/PIckaxe_DriftTrooper/Textures/T_LightningMask_TechArt}
  n17 = SmoothStep(Value=@n16.R) {constMax=0.5}
  n18 = ScalarParameter() {defaultValue=0.1, group=TechArtEmissive, parameterName=LightningDivisorSwitch}
  n19 = ScalarParameter() {defaultValue=3, group=TechArtEmissive, parameterName=SineMulti}
  n20 = Sine(None=Multiply(
    A=MaterialFunctionCall!(Time, Divisor=@n18, Mod) {materialFunction=/Game/Effects/Fort_Effects/Materials/_Functions/MF_ModuloTime}.Unclamped
    B=@n19
   ))
  n21 = Multiply(
   A=LinearInterpolate(A=Multiply(A=@n17) {constB=3}, B=SmoothStep(Value=@n16.G) {constMax=0.9}, Alpha=Step(X=@n20))
   B=ScalarParameter() {defaultValue=2, group=TechArtEmissive, parameterName=ClusterIntensity01}
  )
  n22 = NamedRerouteUsage()
  n23 = Add(
   A=Add(A=Add(A=@n9.R, B=@n15), B=SmoothStep(Value=@n9.R) {constMax=0.8, constMin=0.3})
   B=Clamp(None=Multiply(A=@n21, B=@n22), Max=ScalarParameter() {defaultValue=2.7, group=TechArtEmissive, parameterName=LightningMaxIntensity})
  )
  n24 = NamedRerouteUsage()
  n25 = Add(
   A=ScalarParameter() {defaultValue=0.1, group=TechArtEmissive, parameterName=BackGroundEmissiveIntensity}
   B=Add(B=LinearInterpolate(A=@n23, Alpha=@n24) {constA=1, constB=0.2})
  )
  n26 = VectorParameter() {group=TechArtEmissive, parameterName=OverrideEmissiveTint, defaultValue=<rgba 1 1 1 1>}
  n27 = Multiply(
   A=ChannelMaskParameter(None=@n2.Tech Art Mask) {group=TechArtEmissive, parameterName=OverrideEmissiveChannel}.R
   B=Multiply(A=Multiply(A=@n25, B=@n26.RGB), B=ScalarParameter() {defaultValue=4, group=TechArtEmissive, parameterName=OverrideEmissivIntensity})
  )
  n28 = Multiply(A=@n27) {constB=0.8}
  n29 = MaterialFunctionCall(Color_Morning=@n27, Color_Day=@n27, Color_Evening=@n28, Color_Night=@n28) {materialFunction=&Color_Time_of_Day}
  MaterialFunctionCall(
   Material=SetMaterialAttributes(MaterialAttributes=@n3.MaterialAttributes, Emissive Color=Add(A=@n3.Emissive Color, B=@n29.Result)) {attributeSetTypes [#1]}
   M=@n2.M
   Pre Skinned Local Position=@n2.M
   Distance Blend=@n2.M
   Fresnel=@n2.M
  ) {materialFunction=&MF_CharacterParent_2023_PostFX}
  NamedRerouteDeclaration(None=@n2.Pre skinned Local Position) {name=PreskinnedPos, nodeColor=<rgba 1 0.5625 0 1>}
  n30 = Multiply(A=ComponentMask(None=@n1) {g=true}) {constB=0.001}
  n31 = MaterialFunctionCall(Divisor=ScalarParameter() {defaultValue=-3, group=TechArtEmissive, parameterName=LightningDivisor}) {materialFunction=&MF_ModuloTime}
  n32 = Multiply(A=@n19, B=NamedRerouteDeclaration(None=Add(A=@n30, B=@n31.Unclamped)) {name=TimeScaled, nodeColor=<rgba 0.0312 0 1 1>}) {constA=3}
  NamedRerouteDeclaration(None=ConstantBiasScale(None=Sine(None=@n32))) {name=Time01, nodeColor=<rgba 0.4297 0 1 1>}
  n33 = ScalarParameter() {defaultValue=2, group=TechArtEmissive, parameterName=LineWidth}
  n34 = ComponentMask(None=Add(A=@n1, B=VectorParameter() {group=TechArtEmissive, parameterName=PivotOffset, defaultValue=<rgba -2 0 0 1>}.RGB)) {r=true}
  n35 = Abs(None=@n34)
  NamedRerouteDeclaration(None=OneMinus(None=SmoothStep(Max=@n33, Value=@n35) {constMax=0.5})) {name=CenterLine, nodeColor=<rgba 1 0 0.2812 1>}
  NamedRerouteDeclaration(None=Sign(None=@n34)) {name=HalfMask, nodeColor=<rgba 0.1719 1 0 1>}
  NamedRerouteUsage()
 ]
}
